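; Blueprint interface for selectable elements of the Animation Picker editor tool in AnimInEdTools.
; SetInitialVars passes an element the Control Rig, Level Sequence, per-rig string-array bindings, the owning picker widget and level actor binding proxies.
; SetEditMode takes an edit-mode flag, UpdateSelectionDisplay takes a selected state, and GetEUW returns a value, presumably the picker's editor utility widget.
; It holds only function signatures, leaving behavior to implementers.
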(fn GetEUW
  (return 0))

(fn SetEditMode
  (params (Edit%20Mode :type Boolean)))

(fn SetInitialVars
  (params (Control%20Rig :type Control%20Rig%20Object%20Reference) (Level%20Sequence :type Level%20Sequence%20Object%20Reference) (Control%20Bindings :type Map%20of%20Control%20Rig%20Object%20References%20to%20Animation%20Picker%20String%20Array%20Struct%20Structures) (Anim%20Picker%20EUW :type Animation%20Picker%20Object%20Reference) (Level%20Actors :type Array%20of%20Movie%20Scene%20Binding%20Proxy%20Structures)))

(fn UpdateSelectionDisplay
  (params (Selected :type Boolean))
  (return false))
