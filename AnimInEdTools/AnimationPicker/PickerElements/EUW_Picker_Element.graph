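# Editor utility widget that serves as an element of the Animation Picker tool, with no widget tree of its own, apparently a base for concrete picker elements.
# Set Initial Vars caches the picker, Level Sequence, primary Control Rig, level actors and control bindings.
# Helper functions map the valid primary rig to one or more control names.
# Edit Mode disables the widget while on, and Update Selection Display only stores the selected flag.
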EditorUtilityWidgetBlueprint /AnimInEdTools/AnimationPicker/PickerElements/EUW_Picker_Element
graphs [
 EventGraph {nodes [
   n1 = Event Set Edit Mode!() -> (OutputDelegate, then, Edit Mode:bool) {is=title K2Node_Event}
   > SetIsEnabled(bInIsEnabled=NOT Boolean(A=@n1.Edit Mode))
   > Set Edit Mode(Edit Mode:bool=@n1.Edit Mode) -> (Output_Get:bool)
   n2 = Event Set Initial Vars!() -> (
    OutputDelegate
    then
    Control Rig:Control Rig Object Reference
    Level Sequence:Level Sequence Object Reference
    Control Bindings:Map of Control Rig Object References to Animation Picker String Array Struct Structures
    Anim Picker EUW:Animation Picker Object Reference
    Level Actors:Array of Movie Scene Binding Proxy Structures
   ) {is=title K2Node_Event}
   > Set Primary Control Rig(Primary Control Rig:Control Rig Object Reference=@n2.Control Rig) -> (Output_Get:Control Rig Object Reference)
   > Set Level Sequence(Level Sequence:Level Sequence Object Reference=@n2.Level Sequence) -> (Output_Get:Level Sequence Object Reference)
   > Set Control Bindings(Control Bindings:Map of Control Rig Object References to Animation Picker String Array Struct Structures=@n2.Control Bindings) -> (Output_Get:Map of Control Rig Object References to Animation Picker String Array Struct Structures)
   > Set Anim Picker EUW(Anim Picker EUW:Animation Picker Object Reference=@n2.Anim Picker EUW) -> (Output_Get:Animation Picker Object Reference)
   > Set Level Actors(Level Actors:Array of Movie Scene Binding Proxy Structures=@n2.Level Actors) -> (Output_Get:Array of Movie Scene Binding Proxy Structures)
  ]}
 Get EUW {nodes [
   Get EUW() {is=title K2Node_FunctionEntry}
   > Return Node!(execute, Out EUW:Editor Utility Widget Object Reference=Self-Reference())
  ]}
 Get Edit Mode {nodes [
   Get Edit Mode() {is=title K2Node_FunctionEntry}
   > Return Node!(execute, Edit Mode:bool=Get Edit Mode() -> (Edit Mode:bool) {is=title K2Node_VariableGet})
  ]}
 Set Control for Primary Control Rig {nodes [
   n1 = Set Control for Primary Control Rig!() -> (then, Control Names:String) {is=title K2Node_FunctionEntry}
   n2 = Get Primary Control Rig() -> (Primary Control Rig:Control Rig Object Reference)
   > Is Valid(InputObject=@n2)
   n3 = String To Name(InString=@n1)
   > Add!(
    execute
    TargetMap:Map of Control Rig Object References to Animation Picker String Array Struct Structures=Get Control Bindings() -> (Control Bindings:Map of Control Rig Object References to Animation Picker String Array Struct Structures)
    Key:"Control Rig Object Reference (by ref)"=@n2
    Value:"Animation Picker String Array Struct Structure (by ref)"=Make Animation Picker String Array Struct!(ControlNames_6_#1:Array of Names=Make Array("[0]":Name=@n3) -> (Array:Array of Names)) -> (Animation_Picker_StringArray_Struct:Animation Picker String Array Struct Structure) {is=title K2Node_MakeStruct}
   ) -> (then) {is=title K2Node_CallFunction}
  ]}
 Set Controls for Primary Control Rig {nodes [
   n1 = Set Controls for Primary Control Rig!() -> (then, Control Names:Array of Names) {is=title K2Node_FunctionEntry}
   n2 = Get Primary Control Rig()
   > Is Valid(InputObject=@n2)
   > Add(
    TargetMap=Get Control Bindings()
    Key=@n2
    Value=Make Animation Picker String Array Struct(ControlNames_6_#1=@n1)
   )
  ]}
 Update Selection Display {nodes [
   n1 = Update Selection Display!() -> (then, Selected:bool) {is=title K2Node_FunctionEntry}
   > n2 = Set Selected(Selected:bool=@n1) -> (Output_Get:bool)
   > Return Node!(execute, Out Selected:bool=@n2)
  ]}
]
variables [
 Anim Picker EUW {type=Animation_Picker_C, default=null}
 Control Bindings {default {}}
 Deleteable {type=boolean, default=true}
 Edit Mode {type=boolean}
 Level Actors {type="array<MovieSceneBindingProxy>", default []}
 Level Sequence {type=LevelSequence, default=null}
 Primary Control Rig {type=ControlRig, default=null}
 Selected {type=boolean}
]
widgets {info {rootWidgetClass=null, widgetCount=0}, widgets []}
