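# Editor utility checkbox element for the animation picker that sets and keys a bool control on a Control Rig in Sequencer.
# On toggle, it writes the checked state at the focused Level Sequence's current display-rate frame, inside an undo transaction.
# On construct, an empty or None control name falls back to the widget's object name.
# Pre-construct copies the Label variable into the checkbox's text block.
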EditorUtilityWidgetBlueprint /AnimInEdTools/AnimationPicker/PickerElements/EUW_Picker_Checkbox
graphs [EventGraph {nodes [
   n1 = "On Check State Changed (EditorUtilityCheckBox_34)"!() -> (OutputDelegate, then, bIsChecked:bool) {is=title K2Node_ComponentBoundEvent}
   > BeginTransaction(Description="LOCTEXT(\"Change picker checkbox value\")")
   > n2 = GetFocusedLevelSequence!(execute) -> (then, ReturnValue:Level Sequence Object Reference)
   > n3 = Get Current Local Time!(execute, TimeUnit:EMovieSceneTimeUnit Enum=DisplayRate) -> (then, ReturnValue:Movie Scene Sequence Playback Params Structure) {is=title K2Node_CallFunction}
   > SetLocalControlRigBool!(
    execute
    LevelSequence:Level Sequence Object Reference=@n2
    ControlRig:Control Rig Object Reference=Get Control Rig() -> (Control Rig:Control Rig Object Reference)
    ControlName:Name=Get Control name override() -> (Control name override:Name=None)
    Frame:Frame Number Structure=Make Frame Number!(Value:int=Break Movie Scene Sequence Playback Params!(MovieSceneSequencePlaybackParams:"Movie Scene Sequence Playback Params Structure (by ref)"=@n3) -> (
      Frame_FrameNumber_Value:int
      Frame_SubFrame:float=~
      Time:float=~
      MarkedFrame:String
      Timecode:Timecode Structure
      PositionType:EMovieScenePositionType Enum
      UpdateMethod:EUpdatePositionMethod Enum
      bHasJumped:bool=~
     ) {is=title K2Node_BreakStruct}) -> (FrameNumber:Frame Number Structure) {is=title K2Node_MakeStruct}
    Value:bool=@n1.bIsChecked
    TimeUnit:EMovieSceneTimeUnit Enum=DisplayRate
    bSetKey:bool=true
   ) -> (then)
   > EndTransaction()
   Event Construct()
   n4 = Get Control name override() -> (Control name override=None)
   n5 = "To String (Name)"(InName=@n4)
   > Branch(Condition=OR Boolean(A=IsEmpty(InString=@n5), B="Equal (Name)"(A=@n4)))
   n6 = Self-Reference()
   > Set Control name override(Control name override:Name=String To Name(InString=GetObjectName(Object=@n6))) -> (Output_Get:Name=None)
   Event PreConstruct()
   > Set Text(Text=Get Label() -> (Label:Text), self=Get LabelTextBox() -> (LabelTextBox:Text Object Reference))
   Break Movie Scene Sequence Playback Params!(MovieSceneSequencePlaybackParams=@n3) -> (Frame:Frame Time Structure, Time=~, MarkedFrame, Timecode, PositionType, UpdateMethod, bHasJumped=~)
  ]}]
variables [
 Control name override {type=string, default=None}
 Control Rig {type=ControlRig, default=null}
 Label {type=string, default=""}
 LS {type=LevelSequence, default=null}
]
widgets {
 info {
  parentClass=&/EUW_Picker_Element.EUW_Picker_Element_C
  rootWidgetClass=&CanvasPanel
  widgetCount=3
 }
 widgets [
  CanvasPanel_0 {widgetClassPath=&CanvasPanel, widget=CanvasPanel_0}
  EditorUtilityCheckBox_34 {
   widgetClassPath=&EditorUtilityCheckBox
   bIsVariable=true
   parent=CanvasPanel_0
   widget=EditorUtilityCheckBox_34
   slot {class=CanvasPanelSlot, properties {layoutData {alignment=<0 0>, anchors {maximum=<1 1>, minimum=<0 0>}, offsets=<ltrb 0 0 0 0>}}}
  }
  LabelTextBox {
   widgetClassPath=&TextBlock
   bIsVariable=true
   parent=EditorUtilityCheckBox_34
   widget=LabelTextBox
   slot {class=PanelSlot}
   properties=~
  }
 ]
}
