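# Editor utility slider element for the Animation Picker that drives Control Rig float controls in a Level Sequence, without keying.
# Selection activates it after a short delay, and construction reads its value from the named control.
# Mouse capture opens an undo transaction and a Tick gate that writes bound controls until release.
# Other selected sliders follow by projecting mouse movement onto their rotation-derived forward vectors, scaled by widget size.
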EditorUtilityWidgetBlueprint /AnimInEdTools/AnimationPicker/PickerElements/EUW_Picker_1DSlider
graphs [
 EventGraph {nodes [
   n1 = "On Value Changed (Slider)"!() -> (OutputDelegate, then, Value:float) {is=title K2Node_ComponentBoundEvent}
   > n2 = Branch(Condition=Get Active() -> (Active:bool))
   > Set Current Slider Value(Current Slider Value:float=@n1.Value) -> (Output_Get:float)
   > Set Current Normalized Value(Current Normalized Value:double=Normalize Float!(
     Input:double=Get Current Slider Value() -> (Current Slider Value:float)
     Min:double=Get Min Value() -> (Min Value:float)
     Max:double=Get Max Value() -> (Max Value:float)
     Clamp:bool=true
    ) -> (Output:double=~) {is=title K2Node_MacroInstance}) -> (Output_Get:double)
   SetValue!(
    execute=@n2.else
    self:Slider Object Reference=Get Slider() -> (Slider:Editor Utility Slider Object Reference)
    InValue:float=Get Current Slider Value()
   ) -> (then)
   "On Mouse Capture Begin (Slider)"() {is=title K2Node_ComponentBoundEvent}
   n3 = Get Anim Picker EUW() -> (Anim Picker EUW:Animation Picker Object Reference)
   > n4 = Branch(Condition=OR Boolean!(
     A=Get Doing Right Click!(self:Animation Picker Object Reference=@n3) -> (Doing Right Click:bool)
     B=Get Ctrl Key Down!(self:Animation Picker Object Reference=@n3) -> (Ctrl Key Down:bool)
     C:bool=Get Alt Key Down!(self:Animation Picker Object Reference=@n3) -> (Alt Key Down:bool)
     D:bool=Get Shift Key Down!(self:Animation Picker Object Reference=@n3) -> (Shift Key Down:bool)
    ) -> (ReturnValue))
   BeginTransaction(execute=@n4.else, Context=Change Picker Slider Value)
   > n5 = GetMousePositionOnViewport!(execute, WorldContextObject:Object Reference) -> (then, ReturnValue:Vector 2D Structure)
   > Set Initial Mouse Location(Initial Mouse Location:Vector 2D Structure=@n5) -> (Output_Get:Vector 2D Structure)
   > Clear(TargetArray:Array of EUW Picker 1DSlider Object References=Get Selected Sliders() -> (Selected Sliders:Array of EUW Picker 1DSlider Object References))
   > n6 = Get Selected Picker Buttons!(execute, self:Animation Picker Object Reference=Get Anim Picker EUW()) -> (then, Buttons:Array of EUW Picker Button Object References, All Elements:Array of EUW Picker Element Object References) {is=title K2Node_CallFunction}
   > n7 = Branch(Condition=Is Not Empty!(TargetArray:Array of EUW Picker Element Object References=@n6.All Elements) -> (ReturnValue:bool))
   > n8 = Branch(Condition=Get Selected() -> (Selected:bool))
   > n9 = For Each Loop(Array:Array of EUW Picker Element Object References=@n6.All Elements) -> (Array Element:EUW Picker Element Object Reference)
   Picker Element Clicked!(
    execute=[@n24.then, @n9.Completed]
    self:Animation Picker Object Reference=Get Anim Picker EUW()
    Picker Element:EUW Picker Element Object Reference=Self-Reference()
   ) -> (then)
   > Picker Button Released(self:Animation Picker Object Reference=Get Anim Picker EUW())
   > Update Selection Display!(execute, self:EUW Picker Element Object Reference, Selected:bool=true) -> (then, Out Selected:bool)
   > ExecuteConsoleCommand(Command=Slate.bAllowThrottling 0)
   > n10 = Get Current Local Time!(execute, TimeUnit:EMovieSceneTimeUnit Enum=DisplayRate) -> (then, ReturnValue:Movie Scene Sequence Playback Params Structure) {is=title K2Node_CallFunction}
   > n11 = Set Current Frame Number(Current Frame Number:Frame Number Structure=Break Frame Time!(FrameTime:"Frame Time Structure (by ref)"=Break Movie Scene Sequence Playback Params!(MovieSceneSequencePlaybackParams:"Movie Scene Sequence Playback Params Structure (by ref)"=@n10) -> (
      Frame:Frame Time Structure
      Time:float=~
      MarkedFrame:String
      Timecode:Timecode Structure
      PositionType:EMovieScenePositionType Enum
      UpdateMethod:EUpdatePositionMethod Enum
      bHasJumped:bool=~
     ) {is=title K2Node_BreakStruct}) -> (FrameNumber:Frame Number Structure, SubFrame:float=~) {is=title K2Node_BreakStruct}) -> (Output_Get:Frame Number Structure)
   Gate!(Enter:Exec=@n37.then, Open:Exec=@n11.then, Close:Exec=@n25.then, Toggle:Exec, bStartClosed:bool=true) -> (Exit:Exec) {is=title K2Node_MacroInstance}
   > n12 = Sequence()
   > Set control bindings to current slider value!(execute, self, bSetKey:bool=false) -> (then) {is=title K2Node_CallFunction}
   n13 = GetMousePositionOnViewport(execute=@n12.then_1)
   > Set Current mouse movement vector(Current mouse movement vector:Vector 2D Structure=vector2d - vector2d(A:Vector 2D Structure=@n13, B:Vector 2D Structure=Get Initial Mouse Location() -> (Initial Mouse Location:Vector 2D Structure)) -> (ReturnValue:Vector 2D Structure) {is=title K2Node_PromotableOperator}) -> (Output_Get:Vector 2D Structure)
   > n14 = For Each Loop(Array:Array of EUW Picker 1DSlider Object References=Get Selected Sliders()) -> (Array Element:EUW Picker 1DSlider Object Reference)
   > Set Current sub sliders(Current sub sliders:EUW Picker 1DSlider Object Reference=@n14) -> (Output_Get:EUW Picker 1DSlider Object Reference)
   > n15 = Branch(Condition="Equal (Object)"(A=Get Current sub sliders() -> (Current sub sliders:EUW Picker 1DSlider Object Reference), B=Self-Reference()))
   n16 = Sequence(execute=@n15.else)
   > n17 = "For Each Loop (Map)"!(
    execute:Exec
    MapPin:Map of Control Rig Object References to Animation Picker String Array Struct Structures=Get Control Bindings!(self:EUW Picker Element Object Reference=Get Current sub sliders()) -> (Control Bindings:Map of Control Rig Object References to Animation Picker String Array Struct Structures)
    BreakPin:Exec
   ) -> (then:Exec, KeyPin:Control Rig Object Reference, ValuePin:Animation Picker String Array Struct Structure, CompletedPin:Exec)
   > n18 = For Each Loop(Array:Array of Names=Break Animation Picker String Array Struct!(Animation_Picker_StringArray_Struct:"Animation Picker String Array Struct Structure (by ref)"=@n17.ValuePin) -> (ControlNames_6_#1:Array of Names) {is=title K2Node_BreakStruct}) -> (Array Element:Name)
   n19 = "Dot Product (2D)"!(
    A:Vector 2D Structure=Get Current mouse movement vector() -> (Current mouse movement vector:Vector 2D Structure)
    B:Vector 2D Structure=Get Forward vector!(self:EUW Picker 1DSlider Object Reference=Get Current sub sliders()) -> (Forward vector:Vector 2D Structure)
   ) -> (ReturnValue:double)
   n20 = GetSize!(self:Canvas Panel Slot Object Reference=SlotAsCanvasSlot!(Widget:Widget Object Reference=Self-Reference()) -> (ReturnValue:Canvas Panel Slot Object Reference)) -> (ReturnValue_X:double, ReturnValue_Y:double)
   n21 = Get Current sub sliders()
   n22 = "Clamp (Float)"(
    Value=float + float(
     A=Get Value When start sliding!(self:EUW Picker 1DSlider Object Reference=@n21) -> (Value When start sliding:float)
     B=float / float(A=@n19, B=@n20)
    )
    Min=Get Min Value()
    Max=Get Max Value()
   )
   > SetLocalControlRigFloat!(
    execute
    LevelSequence:Level Sequence Object Reference=Get Level Sequence() -> (Level Sequence:Level Sequence Object Reference)
    ControlRig:Control Rig Object Reference=Get Primary Control Rig() -> (Primary Control Rig:Control Rig Object Reference)
    ControlName:Name=@n18
    Frame:Frame Number Structure=Get Current Frame Number() -> (Current Frame Number:Frame Number Structure)
    Value:float=@n22
    TimeUnit:EMovieSceneTimeUnit Enum=DisplayRate
    bSetKey:bool=false
   ) -> (then)
   SetValue(execute=@n16.then_1, self=Get Slider!(self:EUW Picker 1DSlider Object Reference=Get Current sub sliders()) -> (Slider), InValue=@n22)
   n23 = Cast To EUW_Picker_1DSlider!(execute=@n9.LoopBody, Object=@n9) -> (then, CastFailed, AsEUW Picker 1DSlider:EUW Picker 1DSlider Object Reference) {is=title K2Node_DynamicCast}
   > Add Unique(
    TargetArray:Array of EUW Picker 1DSlider Object References=Get Selected Sliders()
    NewItem:"EUW Picker 1DSlider Object Reference (by ref)"=@n23
   )
   > Set Value When start sliding!(
    execute
    Value When start sliding:float=Get Current Slider Value!(self:EUW Picker 1DSlider Object Reference=@n23) -> (Current Slider Value)
    self:EUW Picker 1DSlider Object Reference=@n23
   ) -> (then, Output_Get:float)
   n24 = Select Controls!(
    execute=[@n8.else, @n7.else]
    self:Animation Picker Object Reference=Get Anim Picker EUW()
    Control Bindings:Map of Control Rig Object References to Animation Picker String Array Struct Structures=Get Control Bindings!() -> (Control Bindings)
    Selection Type:Enum_SelectionType Enum=NewEnumerator0
   ) -> (then)
   "On Mouse Capture End (Slider)"() {is=title K2Node_ComponentBoundEvent}
   > ExecuteConsoleCommand(Command=Slate.bAllowThrottling 1)
   > Set control bindings to current slider value(bSetKey=false)
   > n25 = EndTransaction()
   n26 = Set Slider Active!() -> (OutputDelegate, then, Active:bool=~) {is=title K2Node_CustomEvent}
   > n27 = Branch(Condition=@n26.Active)
   > RetriggerableDelay!(execute, WorldContextObject:Object Reference, Duration=Get Time to hold before active() -> (Time to hold before active:float)) -> (then)
   > Set Active(Active:bool=true) -> (Output_Get:bool)
   Set Active(execute=@n27.else, Active=false)
   n28 = Set Slider Relative Value!() -> (OutputDelegate, then, InValue:float=~) {is=title K2Node_CustomEvent}
   > n29 = Set Current Slider Value(Current Slider Value=Lerp(A=Get Min Value(), B=Get Max Value(), Alpha=@n28.InValue))
   > SetValue(self=Get Slider!() -> (Slider), InValue=@n29)
   Zero Out Slider() {is=title K2Node_CustomEvent}
   > SetValue(self=Get Slider(), InValue=0)
   Event Construct()
   > n30 = Sequence!(execute) -> (then_0, then_1, then_2:Exec)
   Reupdate slider value based on control(execute=@n30.then_1)
   Set Normal Style(
    execute=@n30.then_2
    Normal Style:Slider Style Structure=Get WidgetStyle!(self:Slider Object Reference=Get Slider()) -> (WidgetStyle:Slider Style Structure)
   ) -> (Output_Get:Slider Style Structure)
   n31 = Get Control name override() -> (Control name override:Name=None)
   n32 = "To String (Name)"(InName=@n31)
   n33 = Branch(execute=@n30.then_0, Condition=OR Boolean!(A=IsEmpty(InString=@n32), B="Equal (Name)"(A=@n31)) -> (ReturnValue))
   n34 = Self-Reference()
   > Set Control name override(Control name override:Name=String To Name(InString=GetObjectName(Object=@n34))) -> (Output_Get:Name=None)
   n35 = "For Each Loop (Map)"(execute=@n33.else, MapPin=Get Control Bindings())
   n36 = Get Control name override() -> (Control name override=None)
   > Add!(
    execute
    TargetMap:Map of Control Rig Object References to Animation Picker String Array Struct Structures=Get Control Bindings()
    Key:"Control Rig Object Reference (by ref)"=@n35
    Value:"Animation Picker String Array Struct Structure (by ref)"=Make Animation Picker String Array Struct!(ControlNames_6_#1:Array of Names=Make Array("[0]":Name=@n36) -> (Array:Array of Names)) -> (Animation_Picker_StringArray_Struct:Animation Picker String Array Struct Structure) {is=title K2Node_MakeStruct}
   ) -> (then) {is=title K2Node_CallFunction}
   n37 = Event Tick()
   Event PreConstruct()
   n38 = Get Slider()
   > SetMinValue!(execute, self:Slider Object Reference=@n38, InValue:float=Get Min Value()) -> (then)
   > SetMaxValue!(execute, self:Slider Object Reference=@n38, InValue:float=Get Max Value()) -> (then)
   > Set Value!(execute, Value:float=Get Default Value() -> (Default Value:float), self:Slider Object Reference=Get Slider()) -> (then, Output_Get:float)
   n39 = Self-Reference()
   > Set Forward vector(Forward vector:Vector 2D Structure=GetRotated2D!(
     A:Vector 2D Structure=(X=1,Y=0)
     AngleDeg:float=Get!(self:Widget Object Reference=@n39) -> (
      RenderTransform_Translation:Vector 2D Structure
      RenderTransform_Scale:Vector 2D Structure
      RenderTransform_Shear:Vector 2D Structure
      RenderTransform_Angle:float=~
     ).RenderTransform_Angle
    ) -> (ReturnValue:Vector 2D Structure)) -> (Output_Get:Vector 2D Structure)
   n40 = Event Set Initial Vars!() -> (
    OutputDelegate
    then
    Control Rig:Control Rig Object Reference
    Level Sequence:Level Sequence Object Reference
    Control Bindings:Map of Control Rig Object References to Animation Picker String Array Struct Structures
    Anim Picker EUW:Animation Picker Object Reference
    Level Actors:Array of Movie Scene Binding Proxy Structures
   ) {is=title K2Node_Event}
   > Parent: Set Initial Vars!(
    execute
    Control Rig:Control Rig Object Reference=@n40.Control Rig
    Level Sequence:Level Sequence Object Reference=@n40.Level Sequence
    Control Bindings:Map of Control Rig Object References to Animation Picker String Array Struct Structures=@n40.Control Bindings
    Anim Picker EUW:Animation Picker Object Reference=@n40.Anim Picker EUW
    Level Actors:Array of Movie Scene Binding Proxy Structures=@n40.Level Actors
   ) -> (then) {is=title K2Node_CallParentFunction}
   SetValue!(execute, Input:"Scalable Float Structure (by ref)", NewValue:float=0) -> (then)
  ]}
 Get_Slider_ToolTipText {nodes [
   Get_Slider_ToolTipText() {is=title K2Node_FunctionEntry}
   > Return Node!(execute, ReturnValue:Text="To Text (Name)"!(InName:Name=Get Control name override() -> (Control name override=None)) -> (ReturnValue:Text))
  ]}
 Reupdate slider value based on control {nodes [
   Reupdate slider value based on control() {is=title K2Node_FunctionEntry}
   > n1 = Get Current Local Time(TimeUnit=DisplayRate)
   > Set Current Frame Number(Current Frame Number=Break Frame Time(FrameTime=Break Movie Scene Sequence Playback Params(MovieSceneSequencePlaybackParams=@n1) -> (Time=~, bHasJumped=~)) -> (SubFrame=~))
   > n2 = GetLocalControlRigFloat!(
    execute
    LevelSequence:Level Sequence Object Reference=Get Level Sequence()
    ControlRig:Control Rig Object Reference=Get Primary Control Rig()
    ControlName:Name=Get Control name override() -> (Control name override=None)
    Frame:Frame Number Structure=Get Current Frame Number()
    TimeUnit:EMovieSceneTimeUnit Enum=DisplayRate
   ) -> (then, ReturnValue:float)
   n3 = Get Min Value()
   n4 = Get Max Value()
   > n5 = Set Value(
    Value="Clamp (Float)"(
     Value=@n2
     Min="Min (Float)"(A:double=@n3, B:double=@n4) -> (ReturnValue:double) {is=title K2Node_CommutativeAssociativeBinaryOperator}
     Max="Max (Float)"(A=@n3, B=@n4)
    )
    self=Get Slider()
   )
   > Set Current Slider Value(Current Slider Value=@n5)
  ]}
 Set control bindings to current slider value {nodes [
   n1 = Set control bindings to current slider value!() -> (then, bSetKey:bool) {is=title K2Node_FunctionEntry}
   > n2 = "For Each Loop (Map)"(MapPin=Get Control Bindings())
   > n3 = For Each Loop(Array=Break Animation Picker String Array Struct(Animation_Picker_StringArray_Struct=@n2.ValuePin))
   > SetLocalControlRigFloat(
    LevelSequence=Get Level Sequence()
    ControlRig=Get Primary Control Rig()
    ControlName=@n3
    Frame=Get Current Frame Number()
    Value=Get Current Slider Value!() -> (Current Slider Value)
    TimeUnit=DisplayRate
    bSetKey=@n1
   )
  ]}
 Update Selection Display {nodes [
   n1 = Update Selection Display!() -> (then, Selected:bool) {is=title K2Node_FunctionEntry}
   > Parent: Update Selection Display!(execute, Selected:bool=@n1) -> (then, Out Selected:bool) {is=title K2Node_CallParentFunction}
   > n2 = Sequence!(execute) -> (then_0, then_1)
   > Set Slider Active!(execute, self, Active:bool=Get Selected()) -> (then) {is=title K2Node_CallFunction}
   Set WidgetStyle!(
    execute=@n2.then_1
    WidgetStyle:Slider Style Structure=Select(
     Option 0:Slider Style Structure=Get Normal Style() -> (Normal Style:Slider Style Structure)
     Option 1:Slider Style Structure=Get Selected Style() -> (Selected Style:Slider Style Structure)
     Index=Get Selected()
    ) -> (ReturnValue:Slider Style Structure)
    self:Slider Object Reference=Get Slider()
   ) -> (then, Output_Get:Slider Style Structure)
   > Return Node!(execute, Out Selected:bool=Get Selected())
  ]}
]
variables [
 Active {type=boolean}
 Control name override {type=string, default=None}
 Current Control name {type=string, default=None}
 Current Frame Number {type=FrameNumber, default {value=0}}
 Current mouse movement vector {type=Vector2D, default=<0 0>}
 Current Normalized Value {type=number}
 Current Slider Value {type=number}
 Current sub sliders {type=EUW_Picker_1DSlider_C, default=null}
 Default Value {type=number}
 Forward vector {type=Vector2D, default=<0 0>}
 Initial Mouse Location {type=Vector2D, default=<0 0>}
 Max Value {type=number, default=1}
 Min Value {type=number}
 Normal Style {
  type=SliderStyle
  default {
   barThickness=4
   disabledBarImage {
    bIsDynamicallyLoaded=false
    drawAs=RoundedBox
    imageType=NoImage
    mirroring=NoMirror
    resourceName=None
    resourceObject=null
    tiling=NoTile
    imageSize=<0 0>
    margin=<ltrb 0 0 0 0>
    outlineSettings {
     bUseBrushTransparency=false
     roundingType=FixedRadius
     width=1
     color {colorUseRule=UseColor_ColorTable, specifiedColor=<rgba 0 0 0 0>}
     cornerRadii=<2 2 2 2>
    }
    tintColor {colorUseRule=UseColor_Specified, specifiedColor=<rgba 0.0048 0.0048 0.0048 1>}
    uVRegion {bIsValid=false, max=<0 0>, min=<0 0>}
   }
   disabledThumbImage {
    bIsDynamicallyLoaded=false
    drawAs=RoundedBox
    imageType=NoImage
    mirroring=NoMirror
    resourceName=None
    resourceObject=null
    tiling=NoTile
    imageSize=<8 8>
    margin=<ltrb 0 0 0 0>
    outlineSettings {
     bUseBrushTransparency=false
     roundingType=HalfHeightRadius
     width=0
     color {colorUseRule=UseColor_Specified, specifiedColor=<rgba 0 0 0 0>}
     cornerRadii=<0 0 0 0>
    }
    tintColor {colorUseRule=UseColor_Specified, specifiedColor=<rgba 0.5271 0.5271 0.5271 1>}
    uVRegion {bIsValid=false, max=<0 0>, min=<0 0>}
   }
   hoveredBarImage {
    bIsDynamicallyLoaded=false
    drawAs=RoundedBox
    imageType=NoImage
    mirroring=NoMirror
    resourceName=None
    resourceObject=null
    tiling=NoTile
    imageSize=<0 0>
    margin=<ltrb 0 0 0 0>
    outlineSettings {
     bUseBrushTransparency=false
     roundingType=FixedRadius
     width=1
     color {colorUseRule=UseColor_ColorTable, specifiedColor=<rgba 0 0 0 0>}
     cornerRadii=<2 2 2 2>
    }
    tintColor {colorUseRule=UseColor_Specified, specifiedColor=<rgba 0.349 0.349 0.349 1>}
    uVRegion {bIsValid=false, max=<0 0>, min=<0 0>}
   }
   hoveredThumbImage {
    bIsDynamicallyLoaded=false
    drawAs=RoundedBox
    imageType=NoImage
    mirroring=NoMirror
    resourceName=None
    resourceObject=null
    tiling=NoTile
    imageSize=<8 8>
    margin=<ltrb 0 0 0 0>
    outlineSettings {
     bUseBrushTransparency=false
     roundingType=HalfHeightRadius
     width=0
     color {colorUseRule=UseColor_Specified, specifiedColor=<rgba 0 0 0 0>}
     cornerRadii=<0 0 0 0>
    }
    tintColor {colorUseRule=UseColor_Specified, specifiedColor=<rgba 1 1 1 1>}
    uVRegion {bIsValid=false, max=<0 0>, min=<0 0>}
   }
   normalBarImage {
    bIsDynamicallyLoaded=false
    drawAs=RoundedBox
    imageType=NoImage
    mirroring=NoMirror
    resourceName=None
    resourceObject=null
    tiling=NoTile
    imageSize=<0 0>
    margin=<ltrb 0 0 0 0>
    outlineSettings {
     bUseBrushTransparency=false
     roundingType=FixedRadius
     width=1
     color {colorUseRule=UseColor_ColorTable, specifiedColor=<rgba 0 0 0 0>}
     cornerRadii=<2 2 2 2>
    }
    tintColor {colorUseRule=UseColor_Specified, specifiedColor=<rgba 0.0048 0.0048 0.0048 1>}
    uVRegion {bIsValid=false, max=<0 0>, min=<0 0>}
   }
   normalThumbImage {
    bIsDynamicallyLoaded=false
    drawAs=RoundedBox
    imageType=NoImage
    mirroring=NoMirror
    resourceName=None
    resourceObject=null
    tiling=NoTile
    imageSize=<8 8>
    margin=<ltrb 0 0 0 0>
    outlineSettings {
     bUseBrushTransparency=false
     roundingType=HalfHeightRadius
     width=0
     color {colorUseRule=UseColor_Specified, specifiedColor=<rgba 0 0 0 0>}
     cornerRadii=<0 0 0 0>
    }
    tintColor {colorUseRule=UseColor_Specified, specifiedColor=<rgba 0.5271 0.5271 0.5271 1>}
    uVRegion {bIsValid=false, max=<0 0>, min=<0 0>}
   }
  }
 }
 Selected Sliders {type="array<EUW_Picker_1DSlider_C>", default []}
 Selected Style {
  type=SliderStyle
  default {
   barThickness=4
   disabledBarImage {
    bIsDynamicallyLoaded=false
    drawAs=RoundedBox
    imageType=NoImage
    mirroring=NoMirror
    resourceName=None
    resourceObject=null
    tiling=NoTile
    imageSize=<0 0>
    margin=<ltrb 0 0 0 0>
    outlineSettings {
     bUseBrushTransparency=false
     roundingType=FixedRadius
     width=1
     color {colorUseRule=UseColor_ColorTable, specifiedColor=<rgba 0 0 0 0>}
     cornerRadii=<2 2 2 2>
    }
    tintColor {colorUseRule=UseColor_Specified, specifiedColor=<rgba 0.0048 0.0048 0.0048 1>}
    uVRegion {bIsValid=false, max=<0 0>, min=<0 0>}
   }
   disabledThumbImage {
    bIsDynamicallyLoaded=false
    drawAs=RoundedBox
    imageType=NoImage
    mirroring=NoMirror
    resourceName=None
    resourceObject=null
    tiling=NoTile
    imageSize=<8 8>
    margin=<ltrb 0 0 0 0>
    outlineSettings {
     bUseBrushTransparency=false
     roundingType=HalfHeightRadius
     width=0
     color {colorUseRule=UseColor_Specified, specifiedColor=<rgba 0 0 0 0>}
     cornerRadii=<0 0 0 0>
    }
    tintColor {colorUseRule=UseColor_Specified, specifiedColor=<rgba 0.5271 0.5271 0.5271 1>}
    uVRegion {bIsValid=false, max=<0 0>, min=<0 0>}
   }
   hoveredBarImage {
    bIsDynamicallyLoaded=false
    drawAs=RoundedBox
    imageType=NoImage
    mirroring=NoMirror
    resourceName=None
    resourceObject=null
    tiling=NoTile
    imageSize=<0 0>
    margin=<ltrb 0 0 0 0>
    outlineSettings {
     bUseBrushTransparency=false
     roundingType=FixedRadius
     width=1
     color {colorUseRule=UseColor_Specified, specifiedColor=<rgba 0.8609 0.6311 0.015 1>}
     cornerRadii=<2 2 2 2>
    }
    tintColor {colorUseRule=UseColor_Specified, specifiedColor=<rgba 0.7083 0.7083 0.7083 1>}
    uVRegion {bIsValid=false, max=<0 0>, min=<0 0>}
   }
   hoveredThumbImage {
    bIsDynamicallyLoaded=false
    drawAs=RoundedBox
    imageType=NoImage
    mirroring=NoMirror
    resourceName=None
    resourceObject=null
    tiling=NoTile
    imageSize=<8 8>
    margin=<ltrb 0 0 0 0>
    outlineSettings {
     bUseBrushTransparency=false
     roundingType=HalfHeightRadius
     width=0
     color {colorUseRule=UseColor_Specified, specifiedColor=<rgba 0 0 0 0>}
     cornerRadii=<0 0 0 0>
    }
    tintColor {colorUseRule=UseColor_Specified, specifiedColor=<rgba 1 1 1 1>}
    uVRegion {bIsValid=false, max=<0 0>, min=<0 0>}
   }
   normalBarImage {
    bIsDynamicallyLoaded=false
    drawAs=RoundedBox
    imageType=NoImage
    mirroring=NoMirror
    resourceName=None
    resourceObject=null
    tiling=NoTile
    imageSize=<0 0>
    margin=<ltrb 0 0 0 0>
    outlineSettings {
     bUseBrushTransparency=false
     roundingType=FixedRadius
     width=1
     color {colorUseRule=UseColor_ColorTable, specifiedColor=<rgba 0 0 0 0>}
     cornerRadii=<2 2 2 2>
    }
    tintColor {colorUseRule=UseColor_Specified, specifiedColor=<rgba 0.2031 0.2031 0.2031 1>}
    uVRegion {bIsValid=false, max=<0 0>, min=<0 0>}
   }
   normalThumbImage {
    bIsDynamicallyLoaded=false
    drawAs=RoundedBox
    imageType=NoImage
    mirroring=NoMirror
    resourceName=None
    resourceObject=null
    tiling=NoTile
    imageSize=<8 8>
    margin=<ltrb 0 0 0 0>
    outlineSettings {
     bUseBrushTransparency=false
     roundingType=HalfHeightRadius
     width=0
     color {colorUseRule=UseColor_Specified, specifiedColor=<rgba 0 0 0 0>}
     cornerRadii=<0 0 0 0>
    }
    tintColor {colorUseRule=UseColor_Specified, specifiedColor=<rgba 0.5271 0.5271 0.5271 1>}
    uVRegion {bIsValid=false, max=<0 0>, min=<0 0>}
   }
  }
 }
 Time to hold before active {type=number, default=0.125}
 Value When start sliding {type=number}
]
widgets {
 info {
  parentClass=&/EUW_Picker_Element.EUW_Picker_Element_C
  rootWidgetClass=&Overlay
  widgetCount=2
 }
 widgets [
  Overlay_0 {widgetClassPath=&Overlay, widget=Overlay_0, properties {bOverride_Cursor=true}}
  Slider {
   widgetClassPath=/Script/Blutility.EditorUtilitySlider
   bIsVariable=true
   parent=Overlay_0
   widget=Slider
   properties {
    mouseUsesStep=true
    requiresControllerLock=false
    value=0.2064
    widgetStyle {
     barThickness=4
     disabledBarImage {
      bIsDynamicallyLoaded=false
      drawAs=RoundedBox
      imageType=NoImage
      mirroring=NoMirror
      resourceName=None
      resourceObject=null
      tiling=NoTile
      imageSize=<0 0>
      margin=<ltrb 0 0 0 0>
      outlineSettings {
       bUseBrushTransparency=false
       roundingType=FixedRadius
       width=1
       color {colorUseRule=UseColor_ColorTable, specifiedColor=<rgba 0 0 0 0>}
       cornerRadii=<2 2 2 2>
      }
      tintColor {colorUseRule=UseColor_Specified, specifiedColor=<rgba 0.0048 0.0048 0.0048 1>}
      uVRegion {bIsValid=false, max=<0 0>, min=<0 0>}
     }
     disabledThumbImage {
      bIsDynamicallyLoaded=false
      drawAs=RoundedBox
      imageType=NoImage
      mirroring=NoMirror
      resourceName=None
      resourceObject=null
      tiling=NoTile
      imageSize=<8 8>
      margin=<ltrb 0 0 0 0>
      outlineSettings {
       bUseBrushTransparency=false
       roundingType=HalfHeightRadius
       width=0
       color {colorUseRule=UseColor_Specified, specifiedColor=<rgba 0 0 0 0>}
       cornerRadii=<0 0 0 0>
      }
      tintColor {colorUseRule=UseColor_Specified, specifiedColor=<rgba 0.5271 0.5271 0.5271 1>}
      uVRegion {bIsValid=false, max=<0 0>, min=<0 0>}
     }
     hoveredBarImage {
      bIsDynamicallyLoaded=false
      drawAs=RoundedBox
      imageType=NoImage
      mirroring=NoMirror
      resourceName=None
      resourceObject=null
      tiling=NoTile
      imageSize=<0 0>
      margin=<ltrb 0 0 0 0>
      outlineSettings {
       bUseBrushTransparency=false
       roundingType=FixedRadius
       width=1
       color {colorUseRule=UseColor_ColorTable, specifiedColor=<rgba 0 0 0 0>}
       cornerRadii=<2 2 2 2>
      }
      tintColor {colorUseRule=UseColor_Specified, specifiedColor=<rgba 0.349 0.349 0.349 1>}
      uVRegion {bIsValid=false, max=<0 0>, min=<0 0>}
     }
     hoveredThumbImage {
      bIsDynamicallyLoaded=false
      drawAs=RoundedBox
      imageType=NoImage
      mirroring=NoMirror
      resourceName=None
      resourceObject=null
      tiling=NoTile
      imageSize=<8 8>
      margin=<ltrb 0 0 0 0>
      outlineSettings {
       bUseBrushTransparency=false
       roundingType=HalfHeightRadius
       width=0
       color {colorUseRule=UseColor_Specified, specifiedColor=<rgba 0 0 0 0>}
       cornerRadii=<0 0 0 0>
      }
      tintColor {colorUseRule=UseColor_Specified, specifiedColor=<rgba 1 1 1 1>}
      uVRegion {bIsValid=false, max=<0 0>, min=<0 0>}
     }
     normalBarImage {
      bIsDynamicallyLoaded=false
      drawAs=RoundedBox
      imageType=NoImage
      mirroring=NoMirror
      resourceName=None
      resourceObject=null
      tiling=NoTile
      imageSize=<0 0>
      margin=<ltrb 0 0 0 0>
      outlineSettings {
       bUseBrushTransparency=false
       roundingType=FixedRadius
       width=1
       color {colorUseRule=UseColor_ColorTable, specifiedColor=<rgba 0 0 0 0>}
       cornerRadii=<2 2 2 2>
      }
      tintColor {colorUseRule=UseColor_Specified, specifiedColor=<rgba 0.0048 0.0048 0.0048 1>}
      uVRegion {bIsValid=false, max=<0 0>, min=<0 0>}
     }
     normalThumbImage {
      bIsDynamicallyLoaded=false
      drawAs=RoundedBox
      imageType=NoImage
      mirroring=NoMirror
      resourceName=None
      resourceObject=null
      tiling=NoTile
      imageSize=<8 8>
      margin=<ltrb 0 0 0 0>
      outlineSettings {
       bUseBrushTransparency=false
       roundingType=HalfHeightRadius
       width=0
       color {colorUseRule=UseColor_Specified, specifiedColor=<rgba 0 0 0 0>}
       cornerRadii=<0 0 0 0>
      }
      tintColor {colorUseRule=UseColor_Specified, specifiedColor=<rgba 0.5271 0.5271 0.5271 1>}
      uVRegion {bIsValid=false, max=<0 0>, min=<0 0>}
     }
    }
   }
   slot {class=OverlaySlot, properties {horizontalAlignment=HAlign_Fill, verticalAlignment=VAlign_Fill}}
  }
 ]
}
