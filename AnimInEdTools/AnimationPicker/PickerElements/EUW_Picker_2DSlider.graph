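# Animation Picker element: a 2D slider for editing Vector2D Control Rig controls in a Level Sequence.
# On construct it reads the bound value; dragging maps thumb position into configurable, invertible min/max ranges once selection activates it after a short delay.
# Mouse capture opens an undo transaction and a Tick gate that writes values at the current frame without keying.
# Mouse deltas offset other selected sliders' controls from their starting values.
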EditorUtilityWidgetBlueprint /AnimInEdTools/AnimationPicker/PickerElements/EUW_Picker_2DSlider
graphs [
 Convert to Slider 2D Relative Value {nodes [
   n1 = Convert to Slider 2D Relative Value!() -> (then, Value:Vector 2D Structure) {is=title K2Node_FunctionEntry}
   n2 = BreakVector2D(InVec=@n1)
   > Set Current X(Current X:double=@n2) -> (Output_Get:double)
   > Set Current Y(Current Y:double=@n2.Y) -> (Output_Get:double)
   n3 = MakeVector2D(X=Get Min Value Y() -> (Min Value Y:double), Y=Get Max Value Y() -> (Max Value Y:double))
   n4 = Get Current Y() -> (Current Y:double)
   n5 = MakeVector2D(X=Get Min Value X() -> (Min Value X:double), Y=Get Max Value X() -> (Max Value X:double))
   n6 = Get Current X() -> (Current X:double)
   > Return Node!(
    execute
    ReturnValue:Vector 2D Structure=MakeVector2D(
     X=Remap Float(Input=@n6, OldRange=@n5, NewRange=(X=0,Y=1), Clamp=true) -> (Output=~)
     Y=Remap Float(Input=@n4, OldRange=@n3, NewRange=(X=0,Y=1), Clamp=true) -> (Output=~)
    )
   )
   n7 = Get Min Value X()
   n8 = Get Max Value X()
   Set Relative Min X(Relative Min X:double="Min (Float)"(A:double=@n7, B:double=@n8) -> (ReturnValue:double) {is=title K2Node_CommutativeAssociativeBinaryOperator}) -> (Output_Get:double)
   > Set Relative Max X(Relative Max X:double="Max (Float)"(A=@n7, B=@n8)) -> (Output_Get:double)
   n9 = Get Min Value Y()
   n10 = Get Max Value Y()
   > Set Relative Min Y(Relative Min Y:double="Min (Float)"(A=@n9, B=@n10)) -> (Output_Get:double)
   > Set Relative Max Y(Relative Max Y:double="Max (Float)"(A=@n9, B=@n10)) -> (Output_Get:double)
   n11 = Get Current X()
   n12 = Get Min Value X()
   n13 = Get Max Value X()
   n14 = Get Min Value X()
   n15 = Get Max Value X()
   n16 = Get Current X()
   n17 = Get Min Value X()
   n18 = Get Max Value X()
   MakeVector2D(
    X=SelectFloat(
     A=Normalize Float!(Input:double=@n16, Min:double=@n14, Max:double=@n15, Clamp:bool=true) -> (Output:double=~) {is=title K2Node_MacroInstance}
     B=Normalize Float(Input=@n11, Min=@n13, Max=@n12, Clamp=true) -> (Output=~)
     bPickA="float < float"(A=@n17, B=@n18)
    )
    Y=Normalize Float(Input=Get Current Y(), Min=Get Min Value Y(), Max=Get Max Value Y(), Clamp=true) -> (Output=~)
   )
   n19 = Get Relative Min Y() -> (Relative Min Y:double)
   n20 = Get Relative Max Y() -> (Relative Max Y:double)
   n21 = Get Current Y()
   "Clamp (Float)"(Value=float / float(A=float - float(A=@n21, B=@n19), B=float - float(A=@n20, B=@n19)), Min=Get Relative Min Y(), Max=Get Relative Max Y())
   n22 = Get Current X()
   n23 = Get Relative Min X() -> (Relative Min X:double)
   n24 = Get Relative Max X() -> (Relative Max X:double)
   "Clamp (Float)"(Value=float / float(A=float - float(A=@n22, B=@n23), B=float - float(A=@n24, B=@n23)), Min=Get Relative Min X(), Max=Get Relative Max X())
  ]}
 EventGraph {nodes [
   n1 = "Dot Product (2D)"!(
    A:Vector 2D Structure=Get Current mouse movement vector() -> (Current mouse movement vector:Vector 2D Structure)
    B:Vector 2D Structure=Get Slider X Direction!(self:EUW Picker 2DSlider Object Reference=Get Current Sub Slider() -> (Current Sub Slider:EUW Picker 2DSlider Object Reference)) -> (Slider X Direction:Vector 2D Structure)
   ) -> (ReturnValue:double)
   n2 = Self-Reference()
   n3 = float / float(
    A=GetSize!(self:Canvas Panel Slot Object Reference=SlotAsCanvasSlot!(Widget:Widget Object Reference=@n2) -> (ReturnValue:Canvas Panel Slot Object Reference)) -> (ReturnValue_X:double, ReturnValue_Y:double)
    B=2
   )
   n4 = Get!(self:EUW Picker 2DSlider Object Reference=Get Current Sub Slider()) -> (Value when start sliding_X:double, Value when start sliding_Y:double)
   n5 = "Dot Product (2D)"(
    A=Get Current mouse movement vector()
    B=Get Slider Y Direction!(self:EUW Picker 2DSlider Object Reference=Get Current Sub Slider()) -> (Slider Y Direction:Vector 2D Structure)
   )
   n6 = Self-Reference()
   n7 = float / float(A=GetSize(self=SlotAsCanvasSlot(Widget=@n6)).ReturnValue_Y, B=2)
   n8 = MakeVector2D(X=float + float(A=float / float(A=@n1, B=@n3), B=@n4), Y=float + float(A=@n4.Value when start sliding_Y, B=float / float(A=@n5, B=@n7)))
   n9 = Convert to Slider 2D Relative Value!(Value:Vector 2D Structure=@n8) -> (ReturnValue:Vector 2D Structure) {is=title K2Node_CallFunction}
   PrintString(InString="To String (Vector2d)"!(InVec:Vector 2D Structure=@n9) -> (ReturnValue:String))
   > SetValue!(
    execute
    self:Synth 2DSlider Object Reference=Get Slider!(self:EUW Picker 2DSlider Object Reference=Get Current Sub Slider()) -> (Slider:Synth 2DSlider Object Reference)
    InValue:Vector 2D Structure=@n9
   ) -> (then)
   "On Mouse Capture Begin (Slider)"() {is=title K2Node_ComponentBoundEvent}
   n10 = Get Anim Picker EUW() -> (Anim Picker EUW:Animation Picker Object Reference)
   > n11 = Branch(Condition=OR Boolean!(
     A=Get Doing Right Click!(self:Animation Picker Object Reference=@n10) -> (Doing Right Click:bool)
     B=Get Ctrl Key Down!(self:Animation Picker Object Reference=@n10) -> (Ctrl Key Down:bool)
     C:bool=Get Alt Key Down!(self:Animation Picker Object Reference=@n10) -> (Alt Key Down:bool)
     D:bool=Get Shift Key Down!(self:Animation Picker Object Reference=@n10) -> (Shift Key Down:bool)
    ) -> (ReturnValue))
   BeginTransaction(execute=@n11.else, Context=Change Picker 2D Slider Value)
   > n12 = GetMousePositionOnViewport!(execute, WorldContextObject:Object Reference) -> (then, ReturnValue:Vector 2D Structure)
   > Set Initial Mouse Location(Initial Mouse Location:Vector 2D Structure=@n12) -> (Output_Get:Vector 2D Structure)
   > Clear(TargetArray:Array of EUW Picker 2DSlider Object References=Get Selected Sliders() -> (Selected Sliders:Array of EUW Picker 2DSlider Object References))
   > n13 = Get Selected Picker Buttons!(execute, self:Animation Picker Object Reference=Get Anim Picker EUW()) -> (then, Buttons:Array of EUW Picker Button Object References, All Elements:Array of EUW Picker Element Object References) {is=title K2Node_CallFunction}
   > n14 = Branch(Condition=Is Not Empty!(TargetArray:Array of EUW Picker Element Object References=@n13.All Elements) -> (ReturnValue:bool))
   > n15 = Branch(Condition=Get Selected() -> (Selected:bool))
   > n16 = For Each Loop(Array:Array of EUW Picker Element Object References=@n13.All Elements) -> (Array Element:EUW Picker Element Object Reference)
   Picker Element Clicked!(
    execute=[@n26.then, @n16.Completed]
    self:Animation Picker Object Reference=Get Anim Picker EUW()
    Picker Element:EUW Picker Element Object Reference=Self-Reference()
   ) -> (then)
   > Picker Button Released(self:Animation Picker Object Reference=Get Anim Picker EUW())
   > Update Selection Display!(execute, self:EUW Picker Element Object Reference, Selected:bool=true) -> (then, Out Selected:bool)
   > ExecuteConsoleCommand(Command=Slate.bAllowThrottling 0)
   > n17 = Get Current Local Time!(execute, TimeUnit:EMovieSceneTimeUnit Enum=DisplayRate) -> (then, ReturnValue:Movie Scene Sequence Playback Params Structure) {is=title K2Node_CallFunction}
   > n18 = Set Current Frame Number(Current Frame Number:Frame Number Structure=Break Frame Time!(FrameTime:"Frame Time Structure (by ref)"=Break Movie Scene Sequence Playback Params!(MovieSceneSequencePlaybackParams:"Movie Scene Sequence Playback Params Structure (by ref)"=@n17) -> (
      Frame:Frame Time Structure
      Time:float=~
      MarkedFrame:String
      Timecode:Timecode Structure
      PositionType:EMovieScenePositionType Enum
      UpdateMethod:EUpdatePositionMethod Enum
      bHasJumped:bool=~
     ) {is=title K2Node_BreakStruct}) -> (FrameNumber:Frame Number Structure, SubFrame:float=~) {is=title K2Node_BreakStruct}) -> (Output_Get:Frame Number Structure)
   Gate!(Enter:Exec=@n45.then, Open:Exec=@n18.then, Close:Exec=@n27.then, Toggle:Exec, bStartClosed:bool=true) -> (Exit:Exec) {is=title K2Node_MacroInstance}
   > n19 = Sequence()
   > Set control bindings to current slider value!(execute, self, bSetKey:bool=false) -> (then) {is=title K2Node_CallFunction}
   n20 = GetMousePositionOnViewport(execute=@n19.then_1)
   > Set Current mouse movement vector(Current mouse movement vector:Vector 2D Structure=vector2d - vector2d(A:Vector 2D Structure=@n20, B:Vector 2D Structure=Get Initial Mouse Location() -> (Initial Mouse Location:Vector 2D Structure)) -> (ReturnValue:Vector 2D Structure) {is=title K2Node_PromotableOperator}) -> (Output_Get:Vector 2D Structure)
   > n21 = For Each Loop(Array:Array of EUW Picker 2DSlider Object References=Get Selected Sliders()) -> (Array Element:EUW Picker 2DSlider Object Reference)
   > Set Current Sub Slider(Current Sub Slider:EUW Picker 2DSlider Object Reference=@n21) -> (Output_Get:EUW Picker 2DSlider Object Reference)
   > n22 = Branch(Condition="Equal (Object)"(A=Get Current Sub Slider(), B=Self-Reference()))
   Sequence(execute=@n22.else)
   > n23 = "For Each Loop (Map)"!(
    execute:Exec
    MapPin:Map of Control Rig Object References to Animation Picker String Array Struct Structures=Get Control Bindings!(self:EUW Picker Element Object Reference=Get Current Sub Slider()) -> (Control Bindings:Map of Control Rig Object References to Animation Picker String Array Struct Structures)
    BreakPin:Exec
   ) -> (then:Exec, KeyPin:Control Rig Object Reference, ValuePin:Animation Picker String Array Struct Structure, CompletedPin:Exec)
   > n24 = For Each Loop(Array:Array of Names=Break Animation Picker String Array Struct!(Animation_Picker_StringArray_Struct:"Animation Picker String Array Struct Structure (by ref)"=@n23.ValuePin) -> (ControlNames_6_#1:Array of Names) {is=title K2Node_BreakStruct}) -> (Array Element:Name)
   > SetLocalControlRigVector2D!(
    execute
    LevelSequence:Level Sequence Object Reference=Get Level Sequence() -> (Level Sequence:Level Sequence Object Reference)
    ControlRig:Control Rig Object Reference=Get Primary Control Rig() -> (Primary Control Rig:Control Rig Object Reference)
    ControlName:Name=@n24
    Frame:Frame Number Structure=Get Current Frame Number() -> (Current Frame Number:Frame Number Structure)
    Value:Vector 2D Structure=@n8
    TimeUnit:EMovieSceneTimeUnit Enum=DisplayRate
    bSetKey:bool=false
   ) -> (then)
   n25 = Cast To EUW_Picker_2DSlider!(execute=@n16.LoopBody, Object=@n16) -> (then, CastFailed, AsEUW Picker 2DSlider:EUW Picker 2DSlider Object Reference) {is=title K2Node_DynamicCast}
   > Add Unique(
    TargetArray:Array of EUW Picker 2DSlider Object References=Get Selected Sliders()
    NewItem:"EUW Picker 2DSlider Object Reference (by ref)"=@n25
   )
   > Set Value when start sliding!(
    execute
    Value when start sliding:Vector 2D Structure=Get Current Slider Value!(self:EUW Picker 2DSlider Object Reference=@n25) -> (Current Slider Value:Vector 2D Structure)
    self:EUW Picker 2DSlider Object Reference=@n25
   ) -> (then, Output_Get:Vector 2D Structure)
   n26 = Select Controls!(
    execute=[@n15.else, @n14.else]
    self:Animation Picker Object Reference=Get Anim Picker EUW()
    Control Bindings:Map of Control Rig Object References to Animation Picker String Array Struct Structures=Get Control Bindings!() -> (Control Bindings)
    Selection Type:Enum_SelectionType Enum=NewEnumerator0
   ) -> (then)
   "On Mouse Capture End (Slider)"() {is=title K2Node_ComponentBoundEvent}
   > ExecuteConsoleCommand(Command=Slate.bAllowThrottling 1)
   > Set control bindings to current slider value(bSetKey=false)
   > n27 = EndTransaction()
   n28 = "On Value Changed X (Slider)"!() -> (OutputDelegate, then, Value:float) {is=title K2Node_ComponentBoundEvent}
   > n29 = Branch(Condition=Get Active() -> (Active:bool))
   > n30 = Set Current Value X(Current Value X:float=Lerp(A=Get Min Value X(), B=Get Max Value X(), Alpha=@n28.Value)) -> (Output_Get:float)
   Set Current Slider Value(
    execute=[@n33.then, @n30.then]
    Current Slider Value:Vector 2D Structure=MakeVector2D(X=Get Current Value X() -> (Current Value X:float), Y=Get Current Value Y() -> (Current Value Y:float))
   ) -> (Output_Get:Vector 2D Structure)
   SetValue(
    execute=[@n32.else, @n29.else]
    self=Get Slider!() -> (Slider)
    InValue=Convert to Slider 2D Relative Value(Value=Get Current Slider Value!() -> (Current Slider Value))
   )
   n31 = "On Value Changed Y (Slider)"!() -> (OutputDelegate, then, Value:float) {is=title K2Node_ComponentBoundEvent}
   > n32 = Branch(Condition=Get Active())
   > n33 = Set Current Value Y(Current Value Y:float=Lerp(A=Get Min Value Y(), B=Get Max Value Y(), Alpha=@n31.Value)) -> (Output_Get:float)
   n34 = Set Slider Active!() -> (OutputDelegate, then, Active:bool=~) {is=title K2Node_CustomEvent}
   > n35 = Branch(Condition=@n34.Active)
   > RetriggerableDelay!(execute, WorldContextObject:Object Reference, Duration=Get Time to hold before active() -> (Time to hold before active:float)) -> (then)
   > Set Active(Active:bool=true) -> (Output_Get:bool)
   Set Active(execute=@n35.else, Active=false)
   Zero Out Slider() {is=title K2Node_CustomEvent}
   > SetValue(self=Get Slider(), InValue=(X=0.5,Y=0.5))
   Event Construct()
   > Update Selection Display(Selected=false)
   > n36 = Sequence()
   n37 = Get Control name override() -> (Control name override:Name=None)
   n38 = "To String (Name)"(InName=@n37)
   > Branch(Condition=OR Boolean!(A=IsEmpty(InString=@n38), B="Equal (Name)"(A=@n37)) -> (ReturnValue))
   n39 = Self-Reference()
   > Set Control name override(Control name override:Name=String To Name(InString=GetObjectName(Object=@n39))) -> (Output_Get:Name=None)
   n40 = Get Current Local Time(execute=@n36.then_1, TimeUnit=DisplayRate)
   > Set Current Frame Number(Current Frame Number=Break Frame Time(FrameTime=Break Movie Scene Sequence Playback Params(MovieSceneSequencePlaybackParams=@n40) -> (Time=~, bHasJumped=~)) -> (SubFrame=~))
   > DelayUntilNextTick!(execute, WorldContextObject:Object Reference) -> (then)
   > n41 = "For Each Loop (Map)"(MapPin=Get Control Bindings())
   > n42 = For Each Loop(Array=Break Animation Picker String Array Struct(Animation_Picker_StringArray_Struct=@n41.ValuePin))
   > n43 = GetLocalControlRigVector2D!(
    execute
    LevelSequence:Level Sequence Object Reference=Get Level Sequence()
    ControlRig:Control Rig Object Reference=Get Primary Control Rig()
    ControlName:Name=@n42
    Frame:Frame Number Structure=Get Current Frame Number()
    TimeUnit:EMovieSceneTimeUnit Enum=DisplayRate
   ) -> (then, ReturnValue:Vector 2D Structure)
   > n44 = Set Current Slider Value(Current Slider Value=@n43)
   > SetValue(self=Get Slider(), InValue=Convert to Slider 2D Relative Value(Value=@n44))
   n45 = Event Tick()
   Event PreConstruct()
   > Set WidgetStyle!(
    execute
    WidgetStyle:Synth 2DSlider Style Structure=Get Normal Style() -> (Normal Style:Synth 2DSlider Style Structure)
    self:Synth 2DSlider Object Reference=Get Slider()
   ) -> (then, Output_Get:Synth 2DSlider Style Structure)
   > n46 = Sequence()
   > n47 = Branch(Condition="float < float"(A=Get Min Value X(), B=Get Max Value X()))
   > Set Slider X Direction(Slider X Direction:Vector 2D Structure=(X=1,Y=0)) -> (Output_Get:Vector 2D Structure)
   Set Slider X Direction(execute=@n47.else, Slider X Direction=(X=-1,Y=0))
   n48 = Branch(execute=@n46.then_1, Condition="float < float"(A=Get Min Value Y(), B=Get Max Value Y()))
   > Set Slider Y Direction(Slider Y Direction:Vector 2D Structure=(X=0,Y=1)) -> (Output_Get:Vector 2D Structure)
   Set Slider Y Direction(execute=@n48.else, Slider Y Direction=(X=0,Y=-1))
   n49 = Event Set Initial Vars!() -> (
    OutputDelegate
    then
    Control Rig:Control Rig Object Reference
    Level Sequence:Level Sequence Object Reference
    Control Bindings:Map of Control Rig Object References to Animation Picker String Array Struct Structures
    Anim Picker EUW:Animation Picker Object Reference
    Level Actors:Array of Movie Scene Binding Proxy Structures
   ) {is=title K2Node_Event}
   > Parent: Set Initial Vars!(
    execute
    Control Rig:Control Rig Object Reference=@n49.Control Rig
    Level Sequence:Level Sequence Object Reference=@n49.Level Sequence
    Control Bindings:Map of Control Rig Object References to Animation Picker String Array Struct Structures=@n49.Control Bindings
    Anim Picker EUW:Animation Picker Object Reference=@n49.Anim Picker EUW
    Level Actors:Array of Movie Scene Binding Proxy Structures=@n49.Level Actors
   ) -> (then) {is=title K2Node_CallParentFunction}
  ]}
 Get_Slider_ToolTipText {nodes [
   Get_Slider_ToolTipText() {is=title K2Node_FunctionEntry}
   > Return Node(ReturnValue:Text="To Text (Name)"!(InName:Name=Get Control name override() -> (Control name override=None)) -> (ReturnValue:Text))
  ]}
 Set control bindings to current slider value {nodes [
   n1 = Set control bindings to current slider value!() -> (then, bSetKey:bool=~) {is=title K2Node_FunctionEntry}
   > n2 = "For Each Loop (Map)"(MapPin=Get Control Bindings())
   > n3 = For Each Loop(Array=Break Animation Picker String Array Struct(Animation_Picker_StringArray_Struct=@n2.ValuePin))
   > SetLocalControlRigVector2D(
    LevelSequence=Get Level Sequence()
    ControlRig=Get Primary Control Rig()
    ControlName=@n3
    Frame=Get Current Frame Number()
    Value=Get Current Slider Value()
    TimeUnit=DisplayRate
    bSetKey=@n1
   )
   Return Node!(execute=@n2.CompletedPin)
  ]}
 Update Selection Display {nodes [
   n1 = Update Selection Display!() -> (then, Selected:bool) {is=title K2Node_FunctionEntry}
   > Parent: Update Selection Display!(execute, Selected:bool=@n1) -> (then, Out Selected:bool) {is=title K2Node_CallParentFunction}
   > n2 = Sequence()
   > Set Slider Active!(execute, self, Active:bool=Get Selected()) -> (then) {is=title K2Node_CallFunction}
   Set WidgetStyle(
    execute=@n2.then_1
    WidgetStyle=Select(
     Option 0:Synth 2DSlider Style Structure=Get Normal Style()
     Option 1:Synth 2DSlider Style Structure=Get Selected Style() -> (Selected Style:Synth 2DSlider Style Structure)
     Index=Get Selected()
    ) -> (ReturnValue:Synth 2DSlider Style Structure)
    self=Get Slider()
   )
   > Return Node!(execute, Out Selected:bool=Get Selected())
  ]}
]
variables [
 Active {type=boolean}
 Control name override {type=string, default=None}
 Current Frame Number {type=FrameNumber, default {value=0}}
 Current mouse movement vector {type=Vector2D, default=<0 0>}
 Current Slider Value {type=Vector2D, default=<0 0>}
 Current Sub Slider {type=EUW_Picker_2DSlider_C, default=null}
 Current Value X {type=number}
 Current Value Y {type=number}
 Initial Mouse Location {type=Vector2D, default=<0 0>}
 Max Value X {type=number, default=1}
 Max Value Y {type=number, default=1}
 Min Value X {type=number, default=-1}
 Min Value Y {type=number, default=-1}
 Normal Style {
  type=Synth2DSliderStyle
  default {
   barThickness=2
   backgroundImage {
    bIsDynamicallyLoaded=false
    drawAs=Image
    imageType=NoImage
    mirroring=NoMirror
    resourceName=None
    resourceObject=null
    tiling=NoTile
    imageSize=<32 32>
    margin=<ltrb 0 0 0 0>
    outlineSettings {
     bUseBrushTransparency=false
     roundingType=HalfHeightRadius
     width=0
     color {colorUseRule=UseColor_Specified, specifiedColor=<rgba 0 0 0 0>}
     cornerRadii=<0 0 0 0>
    }
    tintColor {colorUseRule=UseColor_Specified, specifiedColor=<rgba 0 0 0 0.1826>}
    uVRegion {bIsValid=false, max=<0 0>, min=<0 0>}
   }
   disabledBarImage {
    bIsDynamicallyLoaded=false
    drawAs=Image
    imageType=NoImage
    mirroring=NoMirror
    resourceName=None
    resourceObject=null
    tiling=NoTile
    imageSize=<32 32>
    margin=<ltrb 0 0 0 0>
    outlineSettings {
     bUseBrushTransparency=false
     roundingType=HalfHeightRadius
     width=0
     color {colorUseRule=UseColor_Specified, specifiedColor=<rgba 0 0 0 0>}
     cornerRadii=<0 0 0 0>
    }
    tintColor {colorUseRule=UseColor_Specified, specifiedColor=<rgba 1 1 1 1>}
    uVRegion {bIsValid=false, max=<0 0>, min=<0 0>}
   }
   disabledThumbImage {
    bIsDynamicallyLoaded=false
    drawAs=Image
    imageType=NoImage
    mirroring=NoMirror
    resourceName=None
    resourceObject=null
    tiling=NoTile
    imageSize=<32 32>
    margin=<ltrb 0 0 0 0>
    outlineSettings {
     bUseBrushTransparency=false
     roundingType=HalfHeightRadius
     width=0
     color {colorUseRule=UseColor_Specified, specifiedColor=<rgba 0 0 0 0>}
     cornerRadii=<0 0 0 0>
    }
    tintColor {colorUseRule=UseColor_Specified, specifiedColor=<rgba 1 1 1 1>}
    uVRegion {bIsValid=false, max=<0 0>, min=<0 0>}
   }
   normalBarImage {
    bIsDynamicallyLoaded=false
    drawAs=Image
    imageType=NoImage
    mirroring=NoMirror
    resourceName=None
    resourceObject=null
    tiling=NoTile
    imageSize=<32 32>
    margin=<ltrb 0 0 0 0>
    outlineSettings {
     bUseBrushTransparency=false
     roundingType=HalfHeightRadius
     width=0
     color {colorUseRule=UseColor_Specified, specifiedColor=<rgba 0 0 0 0>}
     cornerRadii=<0 0 0 0>
    }
    tintColor {colorUseRule=UseColor_Specified, specifiedColor=<rgba 1 1 1 1>}
    uVRegion {bIsValid=false, max=<0 0>, min=<0 0>}
   }
   normalThumbImage {
    bIsDynamicallyLoaded=false
    drawAs=Image
    imageType=FullColor
    mirroring=NoMirror
    resourceName=None
    resourceObject=/Game/Characters/Player/Common/SharedRigs/Graphics/Circle
    tiling=NoTile
    imageSize=<26 26>
    margin=<ltrb 0 0 0 0>
    outlineSettings {
     bUseBrushTransparency=false
     roundingType=HalfHeightRadius
     width=0
     color {colorUseRule=UseColor_Specified, specifiedColor=<rgba 0 0 0 0>}
     cornerRadii=<0 0 0 0>
    }
    tintColor {colorUseRule=UseColor_Specified, specifiedColor=<rgba 0.2604 0.2604 0.2604 1>}
    uVRegion {bIsValid=false, max=<0 0>, min=<0 0>}
   }
  }
 }
 Selected Sliders {type="array<EUW_Picker_2DSlider_C>", default []}
 Selected Style {
  type=Synth2DSliderStyle
  default {
   barThickness=2
   backgroundImage {
    bIsDynamicallyLoaded=false
    drawAs=Image
    imageType=NoImage
    mirroring=NoMirror
    resourceName=None
    resourceObject=null
    tiling=NoTile
    imageSize=<32 32>
    margin=<ltrb 0 0 0 0>
    outlineSettings {
     bUseBrushTransparency=false
     roundingType=HalfHeightRadius
     width=0
     color {colorUseRule=UseColor_Specified, specifiedColor=<rgba 0 0 0 0>}
     cornerRadii=<0 0 0 0>
    }
    tintColor {colorUseRule=UseColor_Specified, specifiedColor=<rgba 0.1979 0.1979 0.1979 0.1826>}
    uVRegion {bIsValid=false, max=<0 0>, min=<0 0>}
   }
   disabledBarImage {
    bIsDynamicallyLoaded=false
    drawAs=Image
    imageType=NoImage
    mirroring=NoMirror
    resourceName=None
    resourceObject=null
    tiling=NoTile
    imageSize=<32 32>
    margin=<ltrb 0 0 0 0>
    outlineSettings {
     bUseBrushTransparency=false
     roundingType=HalfHeightRadius
     width=0
     color {colorUseRule=UseColor_Specified, specifiedColor=<rgba 0 0 0 0>}
     cornerRadii=<0 0 0 0>
    }
    tintColor {colorUseRule=UseColor_Specified, specifiedColor=<rgba 1 1 1 1>}
    uVRegion {bIsValid=false, max=<0 0>, min=<0 0>}
   }
   disabledThumbImage {
    bIsDynamicallyLoaded=false
    drawAs=Image
    imageType=NoImage
    mirroring=NoMirror
    resourceName=None
    resourceObject=null
    tiling=NoTile
    imageSize=<32 32>
    margin=<ltrb 0 0 0 0>
    outlineSettings {
     bUseBrushTransparency=false
     roundingType=HalfHeightRadius
     width=0
     color {colorUseRule=UseColor_Specified, specifiedColor=<rgba 0 0 0 0>}
     cornerRadii=<0 0 0 0>
    }
    tintColor {colorUseRule=UseColor_Specified, specifiedColor=<rgba 1 1 1 1>}
    uVRegion {bIsValid=false, max=<0 0>, min=<0 0>}
   }
   normalBarImage {
    bIsDynamicallyLoaded=false
    drawAs=Image
    imageType=NoImage
    mirroring=NoMirror
    resourceName=None
    resourceObject=null
    tiling=NoTile
    imageSize=<32 32>
    margin=<ltrb 0 0 0 0>
    outlineSettings {
     bUseBrushTransparency=false
     roundingType=HalfHeightRadius
     width=0
     color {colorUseRule=UseColor_Specified, specifiedColor=<rgba 0 0 0 0>}
     cornerRadii=<0 0 0 0>
    }
    tintColor {colorUseRule=UseColor_Specified, specifiedColor=<rgba 1 1 1 1>}
    uVRegion {bIsValid=false, max=<0 0>, min=<0 0>}
   }
   normalThumbImage {
    bIsDynamicallyLoaded=false
    drawAs=Image
    imageType=FullColor
    mirroring=NoMirror
    resourceName=None
    resourceObject=&Circle
    tiling=NoTile
    imageSize=<32 32>
    margin=<ltrb 0 0 0 0>
    outlineSettings {
     bUseBrushTransparency=false
     roundingType=HalfHeightRadius
     width=0
     color {colorUseRule=UseColor_Specified, specifiedColor=<rgba 0 0 0 0>}
     cornerRadii=<0 0 0 0>
    }
    tintColor {colorUseRule=UseColor_Specified, specifiedColor=<rgba 1 1 1 1>}
    uVRegion {bIsValid=false, max=<0 0>, min=<0 0>}
   }
  }
 }
 Slider X Direction {type=Vector2D, default=<0 0>}
 Slider Y Direction {type=Vector2D, default=<0 0>}
 Time to hold before active {type=number, default=0.25}
 Value when start sliding {type=Vector2D, default=<0 0>}
]
widgets {
 info {
  parentClass=&/EUW_Picker_Element.EUW_Picker_Element_C
  rootWidgetClass=&Overlay
  widgetCount=2
 }
 widgets [
  Overlay_0 {widgetClassPath=&Overlay, widget=Overlay_0, properties {bOverride_Cursor=true}}
  Slider {
   widgetClassPath=/Script/Synthesis.Synth2DSlider
   bIsVariable=true
   parent=Overlay_0
   widget=Slider
   properties {
    valueX=0.5
    valueY=0.5
    widgetStyle {
     barThickness=2
     backgroundImage {
      bIsDynamicallyLoaded=false
      drawAs=Image
      imageType=NoImage
      mirroring=NoMirror
      resourceName=None
      resourceObject=null
      tiling=NoTile
      imageSize=<32 32>
      margin=<ltrb 0 0 0 0>
      outlineSettings {
       bUseBrushTransparency=false
       roundingType=HalfHeightRadius
       width=0
       color {colorUseRule=UseColor_Specified, specifiedColor=<rgba 0 0 0 0>}
       cornerRadii=<0 0 0 0>
      }
      tintColor {colorUseRule=UseColor_Specified, specifiedColor=<rgba 0 0 0 0.1826>}
      uVRegion {bIsValid=false, max=<0 0>, min=<0 0>}
     }
     disabledBarImage {
      bIsDynamicallyLoaded=false
      drawAs=Image
      imageType=NoImage
      mirroring=NoMirror
      resourceName=None
      resourceObject=null
      tiling=NoTile
      imageSize=<32 32>
      margin=<ltrb 0 0 0 0>
      outlineSettings {
       bUseBrushTransparency=false
       roundingType=HalfHeightRadius
       width=0
       color {colorUseRule=UseColor_Specified, specifiedColor=<rgba 0 0 0 0>}
       cornerRadii=<0 0 0 0>
      }
      tintColor {colorUseRule=UseColor_Specified, specifiedColor=<rgba 1 1 1 1>}
      uVRegion {bIsValid=false, max=<0 0>, min=<0 0>}
     }
     disabledThumbImage {
      bIsDynamicallyLoaded=false
      drawAs=Image
      imageType=NoImage
      mirroring=NoMirror
      resourceName=None
      resourceObject=null
      tiling=NoTile
      imageSize=<32 32>
      margin=<ltrb 0 0 0 0>
      outlineSettings {
       bUseBrushTransparency=false
       roundingType=HalfHeightRadius
       width=0
       color {colorUseRule=UseColor_Specified, specifiedColor=<rgba 0 0 0 0>}
       cornerRadii=<0 0 0 0>
      }
      tintColor {colorUseRule=UseColor_Specified, specifiedColor=<rgba 1 1 1 1>}
      uVRegion {bIsValid=false, max=<0 0>, min=<0 0>}
     }
     normalBarImage {
      bIsDynamicallyLoaded=false
      drawAs=Image
      imageType=NoImage
      mirroring=NoMirror
      resourceName=None
      resourceObject=null
      tiling=NoTile
      imageSize=<32 32>
      margin=<ltrb 0 0 0 0>
      outlineSettings {
       bUseBrushTransparency=false
       roundingType=HalfHeightRadius
       width=0
       color {colorUseRule=UseColor_Specified, specifiedColor=<rgba 0 0 0 0>}
       cornerRadii=<0 0 0 0>
      }
      tintColor {colorUseRule=UseColor_Specified, specifiedColor=<rgba 1 1 1 1>}
      uVRegion {bIsValid=false, max=<0 0>, min=<0 0>}
     }
     normalThumbImage {
      bIsDynamicallyLoaded=false
      drawAs=Image
      imageType=FullColor
      mirroring=NoMirror
      resourceName=None
      resourceObject=&Circle
      tiling=NoTile
      imageSize=<32 32>
      margin=<ltrb 0 0 0 0>
      outlineSettings {
       bUseBrushTransparency=false
       roundingType=HalfHeightRadius
       width=0
       color {colorUseRule=UseColor_Specified, specifiedColor=<rgba 0 0 0 0>}
       cornerRadii=<0 0 0 0>
      }
      tintColor {colorUseRule=UseColor_Specified, specifiedColor=<rgba 1 1 1 1>}
      uVRegion {bIsValid=false, max=<0 0>, min=<0 0>}
     }
    }
   }
   slot {class=OverlaySlot, properties {horizontalAlignment=HAlign_Fill, verticalAlignment=VAlign_Fill}}
  }
 ]
}
